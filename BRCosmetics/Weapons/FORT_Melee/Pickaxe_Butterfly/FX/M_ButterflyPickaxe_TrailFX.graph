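# Translucent, two-sided unlit beam-trail material for the Butterfly Pickaxe's trail effect.
# Curve atlas lookups, all using the butterfly pickaxe curve, remap UV gradients, a sphere mask and rotated glass-crack normal green values into trail shape and detail.
# These masks blend blue to near-white, with particle-colored crack highlights added before emissive adjustment.
# Shape mask, UV falloffs and optional particle alpha (UseParticleAlpha) multiply into the apparent opacity.
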Material /BRCosmetics/Weapons/FORT_Melee/Pickaxe_Butterfly/FX/M_ButterflyPickaxe_TrailFX
graph {
 properties {
  bUsedWithBeamTrails=true
  blendMode=BLEND_Translucent
  pixelDepthOffsetMode=PDOM_Legacy
  shadingModel=MSM_Unlit
  twoSided=true
  usedShadingModels=Unlit
 }
 nodes [
  n1 = ParticleColor()
  n2 = Multiply(A=TextureCoordinate(), B=Constant2Vector() {g=0.4, r=2})
  n3 = Constant() {r=-0.25}
  n4 = CurveAtlasRowParameter(CurveTime=TextureSample(UVs=MaterialFunctionCall(UVs=@n2, "Rotation Angle (0-1)"=@n3) {materialFunction=&CustomRotator}.Rotated Values) {samplerType=SAMPLERTYPE_Normal, texture=/Game/Effects/Fort_Effects/Textures/Tile/T_Glass_Cracks_N}.G) {atlas=&AthenaCurveAtlas01, curve=/Game/Athena/LinearColorCurves/Curve_butterflyPickaxe, defaultValue=91}
  n5 = ComponentMask(None=TextureCoordinate()) {r=true}
  n6 = Multiply(A=Multiply(A=OneMinus(None=@n4.R), B=@n4.B), B=Power(Base=OneMinus(None=@n5)) {constExponent=5})
  n7 = Constant3Vector() {constant=<rgba 0.0477 0.5205 1 0>}
  n8 = Constant3Vector() {constant=<rgba 0.9242 1 0.9798 0>}
  n9 = CurveAtlasRowParameter(CurveTime=ComponentMask(None=TextureCoordinate()) {g=true}) {atlas=&AthenaCurveAtlas01, curve=&Curve_butterflyPickaxe, defaultValue=91}
  n10 = TextureCoordinate()
  n11 = Constant2Vector() {g=0.3, r=0.4}
  n12 = CurveAtlasRowParameter(CurveTime=SphereMask(A=Multiply(A=@n10, B=@n11), B=@n11, Radius=Constant() {r=0.5}, Hardness=Constant())) {atlas=&AthenaCurveAtlas01, curve=&Curve_butterflyPickaxe, defaultValue=91}
  n13 = Add(A=Multiply(A=Multiply(A=@n9.R, B=@n12.G) {constA=1}) {constB=3}, B=@n12.G)
  n14 = Add(A=@n6, B=Power(Base=@n13))
  MaterialFunctionCall(Emissive=Add(A=Multiply(A=@n6, B=@n1.RGB), B=LinearInterpolate(A=@n7, B=@n8, Alpha=@n14))) {materialFunction=&MF_FX_EmissiveAdjustments}
  n15 = OneMinus(None=ScalarParameter() {desc="1= Use Particle Alpha, 0= Don't", parameterName=UseParticleAlpha})
  n16 = TextureCoordinate()
  n17 = Multiply(A=OneMinus(None=ComponentMask(None=@n16) {g=true}), B=OneMinus(None=ComponentMask(None=@n16) {r=true}))
  n18 = CurveAtlasRowParameter(CurveTime=ComponentMask(None=TextureCoordinate()) {r=true}) {atlas=&AthenaCurveAtlas01, curve=&Curve_butterflyPickaxe, defaultValue=91}
  Multiply(A=Saturate(None=Add(A=@n1.A, B=@n15)), B=Multiply(A=@n13, B=Multiply(A=@n17, B=@n18.A)))
 ]
}
