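# Unlit additive particle material for the Butterfly pickaxe flare effect, usable on Cascade and Niagara sprites.
# A radial core glow and a rotating spiky flare texture with sine-pulsed contrast are tinted by particle color and passed through MF_FX_EmissiveAdjustments.
# Particle alpha gets a depth fade and, when object and actor positions differ enough vertically, an actor-centered sphere-mask fade.
# The mobile shading path uses plain particle alpha, skipping both fades.
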Material /BRCosmetics/Weapons/FORT_Melee/Pickaxe_Butterfly/FX/M_ButterflyPickaxe_flare
graph {
 properties {
  bUsedWithNiagaraSprites=true
  bUsedWithParticleSprites=true
  blendMode=BLEND_Additive
  pixelDepthOffsetMode=PDOM_Legacy
  shadingModel=MSM_Unlit
  usedShadingModels=Unlit
 }
 nodes [
  n1 = TextureCoordinate()
  n2 = Constant() {r=0.5}
  n3 = SphereMask(A=@n1, B=@n2, Radius=@n2) {hardnessPercent=0}
  n4 = Multiply(A=Power(Base=@n3) {constExponent=5}) {constB=2}
  n5 = ParticleColor()
  n6 = LinearInterpolate(B=@n5.RGB) {constA=1, constAlpha=0.7}
  n7 = Rotator(Coordinate=@n1, Time=Time())
  n8 = Time()
  n9 = Divide(A=Sine(None=Multiply(A=@n8) {constB=0.5}))
  n10 = Add(
   A=Multiply(A=Power(Base=@n3) {constExponent=4}) {constB=0.15}
   B=Power(
    Base=TextureSampleParameter2D(UVs=@n7) {parameterName=FlareMask, texture=/Game/Effects/Fort_Effects/Textures/Flares/T_Flare_Spikeball}.RGB
    Exp=Add(A=@n9) {constB=1.5}
   )
  )
  MaterialFunctionCall(Emissive=Add(A=Multiply(A=@n4, B=@n6), B=Multiply(A=@n10, B=@n5.RGB))) {materialFunction=&MF_FX_EmissiveAdjustments}
  n11 = ObjectPositionWS()
  n12 = ActorPositionWS()
  n13 = MaterialFunctionCall(Vector 3=Multiply(A=Subtract(A=@n11, B=@n12), B=Constant3Vector() {constant=<rgba 0 0 1 0>})) {materialFunction=/Engine/Functions/Engine_MaterialFunctions02/Utility/VectorLength}
  n14 = Constant() {r=105}
  n15 = SphereMask(
   A=WorldPosition().XYZ
   B=ActorPositionWS()
   Radius=ScalarParameter() {defaultValue=110, parameterName=FadeRadius}
   Hardness=ScalarParameter() {parameterName=FadeStrength}
  )
  n16 = Constant() {r=1}
  n17 = DepthFade(
   Opacity=Multiply(A=@n5.A, B=If(A=@n13.V3 Length, B=@n14, "A > B"=@n15, "A == B"=@n15, "A < B"=@n16))
   FadeDistance=ScalarParameter() {defaultValue=30, parameterName=DepthFade}
  )
  Multiply(A=Constant() {r=1}, B=ShadingPathSwitch(Default=@n17, Deferred=@n17, Mobile=@n5.A) {desc=Migrated from FeatureLevelSwitch})
 ]
}
