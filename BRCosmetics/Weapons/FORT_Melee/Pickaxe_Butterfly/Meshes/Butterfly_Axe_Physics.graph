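# Physics asset for the Butterfly pickaxe mesh, a two-handed melee cosmetic with dangling flower and leaf bones.
# One long capsule covers the handle bone, and small capsules cover four flower bones and three leaf bones.
# Each flower and leaf body is constrained directly to the handle rather than chained, with tight swing and twist limits that keep any motion to a slight wobble.
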PhysicsAsset /BRCosmetics/Weapons/FORT_Melee/Pickaxe_Butterfly/Meshes/Butterfly_Axe_Physics
bodies {
 Melee_TwoHanded_Handle [{shapeType=Capsule, length=116.1264, radius=2.2247, center=<1.0601 -0.9385 22.482>}]
 dyn_flower_01 [{shapeType=Capsule, length=2.954, radius=1.7928, center=<1.5447 0.0001 0>, rotation=<rot 90 -0.0396 -0.0396>}]
 dyn_flower_02 [{shapeType=Capsule, length=3.0428, radius=2.1408, center=<1.0461 -0.0001 0>, rotation=<rot 90 0.0074 0.0074>}]
 dyn_flower_03 [{shapeType=Capsule, length=6.4683, radius=2.3157, center=<2.2788 -0.0001 0>, rotation=<rot 90 -0.0264 -0.0263>}]
 dyn_flower_04 [{shapeType=Capsule, length=6.1376, radius=2.8256, center=<2.6165 0.0001 0>, rotation=<rot 90 0.0197 0.0196>}]
 dyn_leaf_01 [{shapeType=Capsule, length=1.8569, radius=1.5527, center=<1.1728 0 0>, rotation=<rot 90 0.0099 0.0099>}]
 dyn_leaf_02 [{shapeType=Capsule, length=2.528, radius=1.9307, center=<1.0007 0.0004 0>, rotation=<rot 90 -0.0099 -0.0099>}]
 dyn_leaf_03 [{shapeType=Capsule, length=1.9797, radius=1.6209, center=<0.7042 -0.0001 0>, rotation=<rot 90 -0.0197 -0.0196>}]
}
constraints [
 {bone1Name=dyn_flower_01, bone2Name=Melee_TwoHanded_Handle, swing1LimitDegrees=20, swing2LimitDegrees=20, twistLimitDegrees=10}
 {bone1Name=dyn_flower_02, bone2Name=Melee_TwoHanded_Handle, swing1LimitDegrees=20, swing2LimitDegrees=20, twistLimitDegrees=10}
 {bone1Name=dyn_flower_03, bone2Name=Melee_TwoHanded_Handle, swing1LimitDegrees=15, swing2LimitDegrees=15, twistLimitDegrees=10}
 {bone1Name=dyn_flower_04, bone2Name=Melee_TwoHanded_Handle, swing1LimitDegrees=15, swing2LimitDegrees=15, twistLimitDegrees=10}
 {bone1Name=dyn_leaf_01, bone2Name=Melee_TwoHanded_Handle, swing1LimitDegrees=20, swing2LimitDegrees=20, twistLimitDegrees=10}
 {bone1Name=dyn_leaf_02, bone2Name=Melee_TwoHanded_Handle, swing1LimitDegrees=20, swing2LimitDegrees=20, twistLimitDegrees=10}
 {bone1Name=dyn_leaf_03, bone2Name=Melee_TwoHanded_Handle, swing1LimitDegrees=20, swing2LimitDegrees=20, twistLimitDegrees=10}
]
